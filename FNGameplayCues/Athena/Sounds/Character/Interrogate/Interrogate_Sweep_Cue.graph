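# Sound cue for the sweep sound of the Athena character Interrogate effect, stored with the gameplay cue sounds.
# A single wave player feeds the output with one sweep wave, with no randomization, mixing or modulation.
# It routes through the third-person enemy player foley sound class with a slight volume boost and a very large max distance.
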SoundCue /FNGameplayCues/Athena/Sounds/Character/Interrogate/Interrogate_Sweep_Cue
root=WavePlayer_0
nodes [{id=WavePlayer_0, properties {soundWaveAssetPtr=/Game/Athena/Sounds/Character/Interrogate/Interrogate_Sweep_01}}]
properties {maxDistance=2097152, soundClassObject=&Fort_Player_Foley_3P_Enemy_Athena, volumeMultiplier=1.2}
